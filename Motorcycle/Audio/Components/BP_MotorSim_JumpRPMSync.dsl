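; Audio motor sim component that layers a jump RPM spike onto the motorcycle engine sound.
; SetJumpCurve stores a curve and scale, then restarts the audio controller's MotorcycleRPMTimeline at a rate based on the curve's end time.
; UpdateRpmSpikeValue stores a normalized progress value.
; OnUpdate samples the curve at that progress and raises runtime RPM to the scaled result when higher, never lowering it.
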(fn OnUpdate
  (params (Input :type Audio%20Motor%20Sim%20Input%20Context%20Structure%20%28by%20ref%29) (RuntimeInfo :type Audio%20Motor%20Sim%20Runtime%20Context%20Structure%20%28by%20ref%29))
  (bind (_bshifting _gear _virtualspeedkmh _rpm _volume _pitch) (Utilities|Struct|BreakAudioMotorSimRuntimeContext RuntimeInfo))
  (bind (_mintime _maxtime) (Math|Curves|GetTimeRange (Variables|Default|GetJumpRpmCurve)))
  (Utilities|Struct|SetmembersinAudioMotorSimRuntimeContext RuntimeInfo _bshifting _gear (Math|Float|Max(Float) _rpm (* (Math|Curves|GetFloatValue (Variables|Default|GetJumpRpmCurve) (Math|Float|MapRangeClamped (Variables|Default|GetValue) 0.0 1.0 _mintime _maxtime)) (Variables|Default|GetJumpRpmScale))) _pitch)
  (return false))

(fn SetJumpCurve
  (params (JumpRpmCurve :type Curve%20Float%20Object%20Reference) (JumpRpmScale :type Float%20%28single-precision%29) (AudioController :type Fort%20Athena%20Vehicle%20Audio%20Controller%20Object%20Reference))
  (bind (_mintime _maxtime) (Math|Curves|GetTimeRange (Variables|Default|GetJumpRpmCurve)))
  (bind _motorcycle_audio_controller (Variables|Default|GetMotorcycle%20Audio%20Controller))
  (bind _motorcyclerpmtimeline (Variables|/Motorcycle/Audio/BP_Vehicles_AudioController_Motorcycle_Parent.BP_Vehicles_AudioController_Motorcycle_Parent_C|GetMotorcycleRPMTimeline _motorcycle_audio_controller))
  (Variables|Default|SetJumpRpmCurve JumpRpmCurve)
  (Variables|Default|SetJumpRpmScale JumpRpmScale)
  (bind _asbp_vehicles_audio_controller_motorcycle_parent (Utilities|Casting|CastToBP_Vehicles_AudioController_Motorcycle_Parent AudioController))
  (Variables|Default|SetMotorcycle%20Audio%20Controller _asbp_vehicles_audio_controller_motorcycle_parent)
  (Components|Timeline|SetPlayRate _motorcyclerpmtimeline (/ 1.0 _maxtime))
  (Components|Timeline|PlayfromStart _motorcyclerpmtimeline))

(fn UpdateRpmSpikeValue
  (params (Value :type Float%20%28double-precision%29))
  (Variables|Default|SetValue Value))
