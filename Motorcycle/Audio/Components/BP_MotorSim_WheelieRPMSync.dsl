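; Audio motor sim component for the motorcycle that, going by its name, syncs simulated engine RPM to wheelies.
; On update, it dots the owner's forward vector with world up to measure pitch and scales it by WheelieRpmScale.
; The result is stored in WheelieRPMInterp and written straight into the motor sim runtime context, with no smoothing here.
; BeginPlay and Tick are empty.
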(event EventBeginPlay)

(event EventTick
  (params (DeltaSeconds :type Float%20%28single-precision%29)))

(fn OnUpdate
  (params (Input :type Audio%20Motor%20Sim%20Input%20Context%20Structure%20%28by%20ref%29) (RuntimeInfo :type Audio%20Motor%20Sim%20Runtime%20Context%20Structure%20%28by%20ref%29))
  (bind _output_get (Variables|Default|SetWheelieRPMInterp (* (Math|Vector|DotProduct (Transformation|GetActorForwardVector (Components|GetOwner)) "0, 0,1.000000") (Variables|Default|GetWheelieRpmScale))))
  (Utilities|Struct|SetmembersinAudioMotorSimRuntimeContext RuntimeInfo _output_get)
  (return false))
