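; Motorcycle vehicle sound component that plays audio for wheelies, stoppies and jumps, bound to the vehicle's trick delegates on init.
; Jumps fade out the charge loop and choose a sound from a threshold map by charge factor.
; Wheelies and jumps temporarily add RPM-sync motor sim components to the audio controller, whose curve-spike events update or remove the jump one.
; A stoppie timer feeds pitch-based RPM to SetManualRPM, which is empty, as are air-trick handlers.
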(event Custom|OnJump
  (params (JumpChargeFactor :type Float%20%28single-precision%29))
  (bind _thresholds (Variables|Jump|Private|GetThresholds))
  (bind _lowjumpindex (Variables|Jump|Private|GetLowJumpIndex))
  (Variables|Jump|Private|SetJump%20Charge%20Factor JumpChargeFactor)
  (Utilities|FlowControl|Sequence
    (:then_0
      (bind _values (Utilities|Map|Values (Variables|Jump|GetJump%20Level%20Sounds)))
      (if (Utilities|Array|IsValidIndex _values)
        (Variables|Jump|Private|SetThresholds _values)
        (bind (_array_element _array_index) (Utilities|Array|ForEachLoopwithBreak _thresholds)
          (:LoopBody
            (if (and (> _array_index 0) (> (Utilities|Array|Get(acopy) _thresholds _array_index) (Variables|Jump|Private|GetJump%20Charge%20Factor)))
              (Variables|Jump|Private|SetLowJumpIndex (- _array_index 1))
              (Variables|Jump|Private|SetHighJumpIndex _array_index)
              (CallFunction|StoponJumpValue)))
          (:Completed
            (if (>= (Variables|Jump|Private|GetJump%20Charge%20Factor) (Utilities|Array|Get(acopy) (Variables|Jump|Private|GetThresholds) (Variables|Jump|Private|GetHighJumpIndex)))
              (Variables|Jump|Private|SetLowJumpIndex (Variables|Jump|Private|GetHighJumpIndex)))
            (bind _keys (Utilities|Map|Keys (Variables|Jump|GetJump%20Level%20Sounds)))
            (Utilities|IsValid (Utilities|Array|Get(acopy) _keys _lowjumpindex)
              (:Is%20Valid
                (Audio|SpawnSoundAttached (Utilities|Array|Get(acopy) _keys _lowjumpindex) (Variables|/Script/Engine.Actor|GetRootComponent (Components|GetOwner)))
                (bind _jumpchargingaudio (Variables|Default|GetJumpChargingAudio)
                  (:then
                    (Audio|Components|Audio|FadeOut _jumpchargingaudio (Variables|Jump|GetJump%20Charging%20Fade%20Out%20Duration) 0.0 "Linear" :!class /Script/Engine.AudioComponent))
                  (:else))))))))
    (:then_1
      (Jump|PlayJumpRpmCurve))))

(event Custom|StopOnJumpValue
  (bind _thresholds (Variables|Jump|Private|GetThresholds))
  (bind _lowjumpindex (Variables|Jump|Private|GetLowJumpIndex))
  (bind (_array_element _array_index) (Utilities|Array|ForEachLoopwithBreak _thresholds)
    (:LoopBody
      (if (and (> _array_index 0) (> (Utilities|Array|Get(acopy) _thresholds _array_index) (Variables|Jump|Private|GetJump%20Charge%20Factor)))
        (Variables|Jump|Private|SetLowJumpIndex (- _array_index 1))
        (Variables|Jump|Private|SetHighJumpIndex _array_index)
        (CallFunction|StoponJumpValue)))
    (:Completed
      (if (>= (Variables|Jump|Private|GetJump%20Charge%20Factor) (Utilities|Array|Get(acopy) (Variables|Jump|Private|GetThresholds) (Variables|Jump|Private|GetHighJumpIndex)))
        (Variables|Jump|Private|SetLowJumpIndex (Variables|Jump|Private|GetHighJumpIndex)))
      (bind _keys (Utilities|Map|Keys (Variables|Jump|GetJump%20Level%20Sounds)))
      (Utilities|IsValid (Utilities|Array|Get(acopy) _keys _lowjumpindex)
        (:Is%20Valid
          (Audio|SpawnSoundAttached (Utilities|Array|Get(acopy) _keys _lowjumpindex) (Variables|/Script/Engine.Actor|GetRootComponent (Components|GetOwner)))
          (bind _jumpchargingaudio (Variables|Default|GetJumpChargingAudio)
            (:then
              (Audio|Components|Audio|FadeOut _jumpchargingaudio (Variables|Jump|GetJump%20Charging%20Fade%20Out%20Duration) 0.0 "Linear" :!class /Script/Engine.AudioComponent))
            (:else)))))))

(event VehicleAudio|EventOnInit
  (|CallParentFunction)
  (Utilities|FlowControl|Sequence
    (:then_0
      (bind _asfort_motorcycle_vehicle (Utilities|Casting|CastToFortMotorcycleVehicle (VehicleSound|GetVehicleActor)))
      ;[1 1 1] Wheelie
      (Motorcycle|BindEventtoOnStartWheelieDelegate _asfort_motorcycle_vehicle (EventDispatchers|CreateEvent :!event OnWheelieStart))
      (Motorcycle|BindEventtoOnLandWheelieDelegate _asfort_motorcycle_vehicle (EventDispatchers|CreateEvent :!event OnWheelieLanded))
      (Motorcycle|BindEventtoOnReleaseWheelieDelegate _asfort_motorcycle_vehicle (EventDispatchers|CreateEvent :!event OnWheelieReleased))
      ;[1 1 1] Stoppie
      (Motorcycle|BindEventtoOnStartStoppieDelegate _asfort_motorcycle_vehicle (EventDispatchers|CreateEvent :!event OnStoppieStart))
      (Motorcycle|BindEventtoOnLandStoppieDelegate _asfort_motorcycle_vehicle (EventDispatchers|CreateEvent :!event OnStoppieLanded))
      ;[1 1 1] Jump
      (Motorcycle|BindEventtoOnJumpDelegate _asfort_motorcycle_vehicle (EventDispatchers|CreateEvent :!event Custom|OnJump))
      (Motorcycle|BindEventtoOnChargingJumpDelegate _asfort_motorcycle_vehicle (EventDispatchers|CreateEvent :!event OnChargingJump))
      (Motorcycle|BindEventtoOnCancelChargingJumpDelegate _asfort_motorcycle_vehicle (EventDispatchers|CreateEvent :!event OnJumpCanceled))
      ;[1 1 1] AirTrick
      (Motorcycle|BindEventtoOnStartAirTrickDelegate _asfort_motorcycle_vehicle (EventDispatchers|CreateEvent :!event OnAirTrick))
      (Motorcycle|BindEventtoOnFinishAirTrickDelegate _asfort_motorcycle_vehicle (EventDispatchers|CreateEvent :!event OnAirTrickEnd)))
    (:then_1
      ;[1 1 1] Motorcycle Audio Ref bindings
      (bind _asbp_vehicles_audio_controller_motorcycle_parent (Utilities|Casting|CastToBP_Vehicles_AudioController_Motorcycle_Parent (VehicleSound|GetAudioController)))
      (Default|BindEventtoStopRpmCurveSpike _asbp_vehicles_audio_controller_motorcycle_parent (EventDispatchers|CreateEvent :!event StopRpmCurveSpike))
      (Default|AssignRpmCurveSpikeUpdate _asbp_vehicles_audio_controller_motorcycle_parent (EventDispatchers|CreateEvent :!event RpmCurveSpikeUpdate)))))

(event VehicleAudio|EventOnShutdown
  (|CallParentFunction)
  ; Unbind from delegates
  (Utilities|FlowControl|Sequence
    (:then_0
      (bind _asfort_motorcycle_vehicle (Utilities|Casting|CastToFortMotorcycleVehicle (VehicleSound|GetVehicleActor)))
      (Motorcycle|UnbindallEventsfromOnStartWheelieDelegate _asfort_motorcycle_vehicle)
      (Motorcycle|UnbindallEventsfromOnLandWheelieDelegate _asfort_motorcycle_vehicle)
      (Motorcycle|UnbindallEventsfromOnReleaseWheelieDelegate _asfort_motorcycle_vehicle)
      (Motorcycle|UnbindallEventsfromOnStartStoppieDelegate _asfort_motorcycle_vehicle)
      (Motorcycle|UnbindallEventsfromOnLandStoppieDelegate _asfort_motorcycle_vehicle)
      (Motorcycle|UnbindallEventsfromOnJumpDelegate _asfort_motorcycle_vehicle)
      (Motorcycle|UnbindallEventsfromOnChargingJumpDelegate _asfort_motorcycle_vehicle)
      (Motorcycle|UnbindallEventsfromOnCancelChargingJumpDelegate _asfort_motorcycle_vehicle)
      (Motorcycle|UnbindallEventsfromOnStartAirTrickDelegate _asfort_motorcycle_vehicle)
      (Vehicle|Boost|UnbindallEventsfromOnBoostFinishedDelegate _asfort_motorcycle_vehicle))
    (:then_1
      (bind _asbp_vehicles_audio_controller_motorcycle_parent (Utilities|Casting|CastToBP_Vehicles_AudioController_Motorcycle_Parent (VehicleSound|GetAudioController)))
      (Default|UnbindallEventsfromStopRpmCurveSpike _asbp_vehicles_audio_controller_motorcycle_parent)
      (Default|UnbindallEventsfromRpmCurveSpikeUpdate _asbp_vehicles_audio_controller_motorcycle_parent))))

(fn OnAirTrick
  (params (TrickType :type EFortMotorcycleAirTrickType%20Enum)))

(fn OnAirTrickEnd)

(fn OnChargingJump
  (params (JumpChargeFactor :type Float%20%28single-precision%29))
  (Variables|Jump|Private|SetJump%20Charge%20Factor JumpChargeFactor)
  (Variables|Default|GetJumpChargingAudio
    (:then)
    (:else
      (bind _jump_charging_sound (Variables|Jump|GetJump%20Charging%20Sound)
        (:then
          (bind _returnvalue (Audio|SpawnSoundAttached _jump_charging_sound (Variables|/Script/Engine.Actor|GetRootComponent (Components|GetOwner))))
          (Variables|Default|SetJumpChargingAudio _returnvalue))
        (:else)))))

(fn OnJumpCanceled
  (bind _jumpchargingaudio (Variables|Default|GetJumpChargingAudio)
    (:then
      (Audio|Components|Audio|FadeOut _jumpchargingaudio (Variables|Jump|GetJump%20Charging%20Fade%20Out%20Duration) 0.0 "Linear" :!class /Script/Engine.AudioComponent))
    (:else)))

(fn OnStoppieLanded
  (if (Variables|Jump|Private|GetIsStoppie)
    (Variables|Jump|Private|SetIsStoppie false)
    (bind _stoppieendsound (Variables|Tricks|Stoppie|GetStoppieEndSound)
      (:then
        (Audio|SpawnSoundAttached _stoppieendsound (Variables|/Script/Engine.Actor|GetRootComponent (Components|GetOwner))))
      (:else
        _))
    (bind _stoppieaudio (Variables|Default|GetStoppieAudio)
      (:then
        (Audio|Components|Audio|FadeOut _stoppieaudio (Variables|Tricks|Stoppie|GetStoppieFadeOutDuration) 0.0 "Linear" :!class /Script/Engine.AudioComponent))
      (:else
        _))
    (Utilities|Time|ClearandInvalidateTimerbyHandle (Variables|Tricks|Stoppie|Private|GetUpdate%20Stoppie%20Handle))))

(fn OnStoppieStart
  (Variables|Jump|Private|SetIsStoppie true)
  (Variables|Default|GetStoppieAudio
    (:then)
    (:else
      (bind _stoppie_sound (Variables|Tricks|Stoppie|GetStoppie%20Sound)
        (:then
          (bind _returnvalue (Audio|SpawnSoundAttached _stoppie_sound (Variables|/Script/Engine.Actor|GetRootComponent (Components|GetOwner))))
          (Variables|Default|SetStoppieAudio _returnvalue)
          (bind _returnvalue_2 (Utilities|Time|SetTimerbyEvent (EventDispatchers|CreateEvent :!event UpdateStoppie) 0.01 true))
          (Variables|Tricks|Stoppie|Private|SetUpdate%20Stoppie%20Handle _returnvalue_2))
        (:else)))))

(fn OnWheelieLanded
  (bind _wheeliemotorsimcomp (Variables|Default|GetWheelieMotorSimComp))
  (Utilities|FlowControl|Sequence
    (:then_0
      (bind _wheelieendsound (Variables|Tricks|Wheelie|GetWheelieEndSound)
        (:then
          (Audio|SpawnSoundAttached _wheelieendsound (VehicleSound|GetRootComponent)))
        (:else
          _))
      (bind _wheelieaudio (Variables|Default|GetWheelieAudio)
        (:then
          (Audio|Components|Audio|FadeOut _wheelieaudio (Variables|Tricks|Wheelie|GetWheelie%20Fade%20Out%20Duration) 0.0 "Linear" :!class /Script/Engine.AudioComponent))
        (:else)))
    (:then_1
      (Variables|Default|GetWheelieMotorSimComp
        (:then
          (bind _asbp_vehicle_cosmetics_audio_controller_parent (Utilities|Casting|CastToBP_VehicleCosmetics_AudioController_Parent (VehicleSound|GetAudioController)))
          (MotorModel|RemoveMotorSimComponent _asbp_vehicle_cosmetics_audio_controller_parent _wheeliemotorsimcomp :!class /VehicleCosmeticsAudio/Controllers/BP_VehicleCosmetics_AudioController_Parent.BP_VehicleCosmetics_AudioController_Parent_C)
          (Components|RemoveArbitraryComponent _asbp_vehicle_cosmetics_audio_controller_parent (Utilities|GetClass _wheeliemotorsimcomp))
          (Components|DestroyComponent _wheeliemotorsimcomp)
          (Variables|Default|SetWheelieMotorSimComp 0))
        (:else)))))

(fn OnWheelieReleased
  (bind _wheeliemotorsimcomp (Variables|Default|GetWheelieMotorSimComp))
  (Variables|Default|GetWheelieMotorSimComp
    (:then
      (bind _asbp_vehicle_cosmetics_audio_controller_parent (Utilities|Casting|CastToBP_VehicleCosmetics_AudioController_Parent (VehicleSound|GetAudioController)))
      (MotorModel|RemoveMotorSimComponent _asbp_vehicle_cosmetics_audio_controller_parent _wheeliemotorsimcomp :!class /VehicleCosmeticsAudio/Controllers/BP_VehicleCosmetics_AudioController_Parent.BP_VehicleCosmetics_AudioController_Parent_C)
      (Components|RemoveArbitraryComponent _asbp_vehicle_cosmetics_audio_controller_parent (Utilities|GetClass _wheeliemotorsimcomp))
      (Components|DestroyComponent _wheeliemotorsimcomp)
      (Variables|Default|SetWheelieMotorSimComp 0))
    (:else)))

(fn OnWheelieStart
  (bind _returnvalue (VehicleSound|GetAudioController))
  (Utilities|FlowControl|Sequence
    (:then_0
      (Variables|Default|GetWheelieAudio
        (:then)
        (:else
          (bind _wheeliesound (Variables|Tricks|Wheelie|GetWheelieSound)
            (:then
              (bind _returnvalue_2 (Audio|SpawnSoundAttached _wheeliesound (Variables|/Script/Engine.Actor|GetRootComponent (Components|GetOwner))))
              (Variables|Default|SetWheelieAudio _returnvalue_2))
            (:else)))))
    (:then_1
      (Variables|Default|GetWheelieMotorSimComp
        (:then)
        (:else
          (bind _asbp_vehicle_cosmetics_audio_controller_parent (Utilities|Casting|CastToBP_VehicleCosmetics_AudioController_Parent _returnvalue))
          (bind _returnvalue_3 (Game|AddComponentbyClass _returnvalue "/Motorcycle/Audio/Components/BP_MotorSim_WheelieRPMSync.BP_MotorSim_WheelieRPMSync_C"))
          (bind _output_get (Variables|Default|SetWheelieMotorSimComp _returnvalue_3))
          (MotorModel|AddMotorSimComponentOrder _asbp_vehicle_cosmetics_audio_controller_parent _output_get))))))

(fn PlayJumpRpmCurve
  (bind _returnvalue (VehicleSound|GetAudioController))
  (Utilities|IsValid (Variables|Jump|GetJumpRpmCurve)
    (:Is%20Valid
      (Variables|Default|GetJumpMotorSimComp
        (:then)
        (:else
          (bind _asbp_vehicle_cosmetics_audio_controller_parent (Utilities|Casting|CastToBP_VehicleCosmetics_AudioController_Parent _returnvalue))
          (bind _returnvalue_2 (Game|AddComponentbyClass _returnvalue "/Motorcycle/Audio/Components/BP_MotorSim_JumpRPMSync.BP_MotorSim_JumpRPMSync_C"))
          (bind _output_get (Variables|Default|SetJumpMotorSimComp _returnvalue_2))
          (Class|BPMotorSimJumpRPMSync|SetJumpCurve (Variables|Default|GetJumpMotorSimComp) (Variables|Jump|GetJumpRpmCurve) (Variables|Jump|GetJumpRpmScale) (VehicleSound|GetAudioController))
          (MotorModel|AddMotorSimComponentOrder _asbp_vehicle_cosmetics_audio_controller_parent _output_get))))))

(fn RpmCurveSpikeUpdate
  (params (Value :type Float%20%28double-precision%29))
  (bind _jumpmotorsimcomp (Variables|Default|GetJumpMotorSimComp)
    (:then
      (Class|BPMotorSimJumpRPMSync|UpdateRpmSpikeValue _jumpmotorsimcomp Value))
    (:else)))

(fn SetManualRPM
  (params (InRPM :type Float%20%28single-precision%29)))

(fn StopRpmCurveSpike
  (bind _jumpmotorsimcomp (Variables|Default|GetJumpMotorSimComp))
  (Variables|Default|GetJumpMotorSimComp
    (:then
      (bind _asbp_vehicle_cosmetics_audio_controller_parent (Utilities|Casting|CastToBP_VehicleCosmetics_AudioController_Parent (VehicleSound|GetAudioController)))
      (MotorModel|RemoveMotorSimComponent _asbp_vehicle_cosmetics_audio_controller_parent _jumpmotorsimcomp :!class /VehicleCosmeticsAudio/Controllers/BP_VehicleCosmetics_AudioController_Parent.BP_VehicleCosmetics_AudioController_Parent_C)
      (Components|RemoveArbitraryComponent _asbp_vehicle_cosmetics_audio_controller_parent (Utilities|GetClass _jumpmotorsimcomp))
      (Components|DestroyComponent _jumpmotorsimcomp)
      (Variables|Default|SetJumpMotorSimComp 0))
    (:else)))

(fn UpdateStoppie
  (if (Variables|Tricks|Stoppie|GetStoppieRpmSpike)
    (CallFunction|SetManualRPM (* (Math|Vector|DotProduct (Vehicle|GetVehicleForwardVector (VehicleSound|GetVehicleActor)) "0, 0,-1.000000") (Variables|Tricks|Stoppie|GetStoppieRpmScale)))
    (else
      (Utilities|Time|ClearandInvalidateTimerbyHandle (Variables|Tricks|Stoppie|Private|GetUpdate%20Stoppie%20Handle)))))
